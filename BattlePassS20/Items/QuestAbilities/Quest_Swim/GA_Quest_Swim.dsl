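; Battle Pass S20 gameplay ability that reports a swim quest stat event when triggered by a gameplay event.
; On activation it calls the parent handler, then sends a complex custom stat event through the owning player's quest manager.
; The event carries the output of a Quest Stat Event Tags setter, an empty second tag container and a count of one.
; The ability ends immediately afterward, with no swim detection of its own.
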(event Ability|EventActivateAbilityFromEvent
  (params (EventData :type Gameplay%20Event%20Data%20Structure%20%28by%20ref%29))
  (|CallParentFunction EventData)
  (bind _output_get (Variables|Default|SetQuest%20Stat%20Event%20Tags))
  (Quest|SendComplexCustomStatEvent (Variables|GAVars|Data|GetOwningPlayerQuestManager) 0 _output_get "(GameplayTags=)" 1)
  (Ability|EndAbility))
